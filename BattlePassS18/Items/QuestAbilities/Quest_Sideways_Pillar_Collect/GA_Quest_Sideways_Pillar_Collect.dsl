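; Gameplay ability for a Sideways pillar collection quest, crediting harvest hits on pillars.
; On an event activation, it checks the event target for the configured pillar tag, then the avatar's owned tags for being inside the Sideways with a harvesting tool equipped.
; When both pass, it sends a complex custom stat event to the owning player's quest manager.
; The ability ends on every path, with debug-gated prints tracing each branch.
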(event Ability|EventActivateAbilityFromEvent
  (params (EventData :type Gameplay%20Event%20Data%20Structure%20%28by%20ref%29))
  (bind (_eventtag _instigator _target) (Utilities|Struct|BreakGameplayEventData EventData))
  (|CallParentFunction EventData)
  (Development|PrintString "@@@ ActivateAbilityFromEvent [GA_Quest_Sideways_Pillar_Collect]: CALLED!" false (Variables|Debug|GetbDebugAbility))
  (bind _returnvalue (GameplayTags|HasMatchingGameplayTag(Message) _target (Variables|Default|GetTargetTag)))
  (if _returnvalue
    (Development|PrintString "@@@@@@ ActivateAbilityFromEvent [GA_Quest_Sideways_Pillar_Collect]: Harvested Target is a sideways pillar!" false (Variables|Debug|GetbDebugAbility))
    (if (GameplayTags|HasAllTags (GameplayTags|GetOwnedGameplayTags (Ability|GetAbilitySystemComponent (Ability|GetAvatarActorfromActorInfo)) :!class /Script/GameplayTags.GameplayTagAssetInterface) (GameplayTags|MakeLiteralGameplayTagContainer "(GameplayTags=((TagName=\"Granted.Athena.Sideways.InsideSideways\"),(TagName=\"Granted.ItemEquipped.Melee.Harvest\")))"))
      (Development|PrintString "@@@@@@@@@ ActivateAbilityFromEvent [GA_Quest_Sideways_Pillar_Collect]: Source IS INSIDE sideways!" false (Variables|Debug|GetbDebugAbility))
      (Quest|SendComplexCustomStatEvent (Variables|GAVars|Data|GetOwningPlayerQuestManager) 0 (Variables|Default|GetQuestStatEventTags) "(GameplayTags=)" 5)
      (else
        (Development|PrintString "@@@@@@@@@ ActivateAbilityFromEvent [GA_Quest_Sideways_Pillar_Collect]: Source IS NOT inside sideways!" false (Variables|Debug|GetbDebugAbility))))
    (Ability|EndAbility)
    (else
      (Development|PrintString "@@@@@@ ActivateAbilityFromEvent [GA_Quest_Sideways_Pillar_Collect]: Harvested target is not a sideways pillar" false (Variables|Debug|GetbDebugAbility))
      (Ability|EndAbility))))
